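# Material function that gives the egg prop an icy look, returning material attributes with emissive color, roughness, refraction and normal.
# Emissive combines a tinted sky cubemap reflection, a sun/moon specular highlight, egg diffuse tinted by shared winter ice color with Fresnel glow, and an edge-mask highlight.
# The Crack dynamic parameter thresholds a tiled texture, revealing glass cracks in color and in the egg normal, which is blended with ice-scratch detail.
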MaterialFunction /Game/Environments/Props/Egg/Materials/MF_Egg_Ice
graph {nodes [
  n1 = ScalarParameter() {defaultValue=0.02, group=Dragon_Scales, parameterName=Edge Highlight Amount}
  n2 = FunctionInput() {inputName=Egg_Edgemask}
  n3 = ScalarParameter() {defaultValue=0.3, group=Dragon_Scales, parameterName=Ice_Multiply}
  n4 = TextureSampleParameterCube!(UVs=ReflectionVectorWS() {bNormalizeCustomWorldNormal=true}, Apply View MipBias) {
   parameterName=MaterialExpressionTextureSampleParameterCube_0
   samplerType=SAMPLERTYPE_LinearColor
   texture=/Game/Packages/Fortress_Sky/TexturesHDR/T_AthenaSkylight
  }
  n5 = VectorParameter() {parameterName=Ice_Cube Map Multiplier, defaultValue=<rgba 0.3282 0.5416 0.7 0>}
  n6 = MaterialFunctionCall() {materialFunction=/Engine/Functions/GetPostProcessFunctions/GetAmbientCubemapTint}
  n7 = Desaturation(
   None=Multiply(A=Multiply(A=@n4.RGB, B=@n5.RGB) {constB=1.5}, B=ComponentMask(None=@n6.AmbientCubemapTint) {b=true, g=true, r=true})
   Fraction=Constant() {r=0.5}
  )
  n8 = CollectionParameter() {
   collection=&FortniteMaterialParameters
   parameterName=SunAndMoonModelDirectionalVector
  }
  n9 = DotProduct(A=ComponentMask(None=@n8) {b=true, g=true, r=true}, B=ReflectionVectorWS())
  n10 = Multiply(
   A=Power(Base=Clamp(None=@n9), Exp=ScalarParameter() {defaultValue=70, group=Dragon_Scales, parameterName=Ice_Spec Highlight Power}) {constExponent=20}
   B=ScalarParameter() {defaultValue=3, group=Dragon_Scales, parameterName=Ice_Spec Intense}
  )
  n11 = FunctionInput() {inputName=Egg_Alpha, inputType=FunctionInput_Scalar}
  n12 = Clamp(None=LinearInterpolate(Alpha=ComponentMask(None=@n11) {a=true}) {constA=0.72, constB=-3})
  n13 = ScalarParameter() {defaultValue=0.6, group=Dragon_Scales, parameterName=DiffuseMultCubemap}
  n14 = Add(A=ComponentMask(None=FunctionInput() {inputName=Egg_Diffuse}) {b=true, g=true, r=true}) {constB=0.1}
  n15 = Add(
   A=MaterialFunctionCall() {materialFunction=/Game/Athena/WinterArt/Materials/MF_SharedWinter}.IceColor
   B=Multiply(
    A=VectorParameter() {group=Dragon_Scales, parameterName=Ice_EdgeGlow, defaultValue=<rgba 0.6206 1.4249 2 0>}.RGB
    B=Fresnel() {baseReflectFraction=0, exponent=2}
   )
  )
  n16 = TextureCoordinate() {uTiling=1.5, vTiling=1.5}
  n17 = TextureSample(UVs=@n16) {texture=/Game/Effects/Fort_Effects/Textures/Tile/T_Glass_Cracks_M}
  n18 = OneMinus(None=@n17.G)
  n19 = Floor(None=Add(
    A=TextureSample(UVs=@n16) {texture=&T_Smooth_tile}.G
    B=DynamicParameter() {defaultValue=<rgba 0 1 1 1>, paramNames [Crack,WobbleTime,Param3,Param4]}.Crack
   ))
  n20 = Add(
   A=Multiply(A=Add(A=@n7, B=@n10), B=LinearInterpolate(B=@n12, Alpha=@n13) {constA=1})
   B=Multiply(A=Multiply(A=@n14, B=@n15), B=LinearInterpolate(B=@n18, Alpha=@n19) {constA=1})
  )
  n21 = FunctionInput() {inputName=Egg_Normal}
  n22 = TextureSample(UVs=@n16) {samplerType=SAMPLERTYPE_Normal, texture=/Game/Effects/Fort_Effects/Textures/Tile/T_Glass_Cracks_N}
  n23 = Constant3Vector() {constant=<rgba 3 3 0 0>}
  n24 = Multiply(A=Add(A=Multiply(A=@n17.R) {constB=0.5}, B=Multiply(A=@n22.RGB, B=@n23)), B=@n19)
  n25 = TextureSample(UVs=Multiply(A=ScalarParameter() {defaultValue=0.3, group=Dragon_Scales, parameterName=Ice Normal UV scale}, B=TextureCoordinate())) {samplerType=SAMPLERTYPE_Normal, texture=/Game/Global/Textures/Default/Tileable_Patterns/T_IceScratch1_N}
  n26 = ScalarParameter() {defaultValue=0.8, group=Dragon_Scales, parameterName=Ice_NormalFlatness}
  n27 = SetMaterialAttributes!(
   MaterialAttributes
   EmissiveColor=Add(A=Multiply(A=@n1, B=@n2), B=Multiply(A=@n3, B=@n20))
   Roughness=Constant() {r=0.7}
   Refraction=ScalarParameter() {defaultValue=0.97, group=Dragon_Scales, parameterName=Ice_Refraction}
   Normal=MaterialFunctionCall(
    BaseNormal=Add(A=@n21, B=@n24)
    AdditionalNormal=MaterialFunctionCall(Normal=@n25.RGB, Flatness=@n26) {materialFunction=&FlattenNormal}.Result
   ) {materialFunction=&BlendAngleCorrectedNormals}.Result
   WorldPositionOffset
  ) {attributeSetTypes [
    #1
    #2
    #3
    #4
    #5
   ]}
  FunctionOutput(None=@n27)
 ]}
