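# Material function that makes a dragon image appear inside an egg prop by drawing it on a virtual plane.
# The plane turns about the vertical axis to face the camera, using the flattened camera-to-object direction and its cross product with up.
# A position offset from the object sets its center, and object radius times a scale parameter sets its size.
# The sampled texture, multiplied by a 0-1 UV mask, becomes BaseColor.
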MaterialFunction /Game/Environments/Props/Egg/Materials/MF_Egg_Inside
graph {nodes [
  n1 = ObjectPositionWS()
  n2 = CameraPositionWS()
  n3 = Normalize(VectorInput=Multiply(A=Subtract(A=@n1, B=@n2), B=Constant3Vector() {constant=<rgba 1 1 0 0>}))
  n4 = CrossProduct(A=@n3, B=Constant3Vector() {constant=<rgba 0 0 1 0>})
  n5 = Multiply(A=ObjectRadius(), B=ScalarParameter() {defaultValue=1.5, group=Dragon_Inside_Egg, parameterName=Dragon_Inside_Scale})
  n6 = MaterialFunctionCall(
   Plane Normal Axis=@n3
   Plane X Axis=Multiply(A=Normalize(VectorInput=@n4)) {constB=-1}
   Plane Center=Add(A=VectorParameter() {group=Dragon_Inside_Egg, parameterName=Dragon Position, defaultValue=<rgba 0 0 1.8585 1>}.RGB, B=ObjectPositionWS())
   UV Scale=AppendVector(A=@n5, B=@n5)
  ) {materialFunction=&VirtualPlaneCoordinates}
  n7 = TextureSampleParameter2D(UVs=@n6.UVs) {parameterName=Dragon_Alpha, texture=/Game/Environments/Props/Egg/Textures/T_Egg_Inside_D}
  n8 = MaterialFunctionCall(UVs=@n6.UVs) {materialFunction=&ScaleUVsByCenter}
  FunctionOutput(None=SetMaterialAttributes!(MaterialAttributes, BaseColor=Multiply(A=@n7.RGB, B=@n8.0-1 mask)) {attributeSetTypes [#1]})
 ]}
