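# Nanite-enabled material for an egg prop, layering an ice function, a tinted shell and snow.
# The shell uses Fresnel to blend contrast-adjusted MF_Egg_Inside color toward an outer tint, adds edge highlights, and blends over MF_Egg_Ice by inverted mask alpha.
# Snow covers upward-facing normals through a height-lerped mask, broken up by world-XY swirl noise that changes scale with depth.
# The result passes through MF_CreativeMoveTool.
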Material /Game/Environments/Props/Egg/Materials/M_Egg
graph {
 properties {
  bUseMaterialAttributes=true
  bUsedWithNanite=true
  pixelDepthOffsetMode=PDOM_Legacy
  refractionMethod=RM_IndexOfRefraction
  usedShadingModels=Default Lit
 }
 nodes [
  n1 = TextureSampleParameter2D() {group=Textures, parameterName=Egg_D, samplerType=SAMPLERTYPE_LinearColor, texture=/Game/Environments/Props/Egg/Textures/T_Egg_D}
  n2 = TextureSampleParameter2D() {group=Textures, parameterName=Egg_N, samplerType=SAMPLERTYPE_Normal, texture=/Game/Environments/Props/Egg/Textures/T_Egg_N}
  n3 = TextureSampleParameter2D() {group=Textures, parameterName=Egg_Edges, texture=/Game/Environments/Props/Egg/Textures/T_Egg_Edge}
  n4 = MaterialFunctionCall!(Egg_Alpha=@n1.A, Egg_Normal=@n2.RGB, Egg_Edgemask=@n3.RGB, Egg_Diffuse=@n1.A) {materialFunction=&/MF_Egg_Ice}
  n5 = Multiply(A=ScalarParameter() {defaultValue=1, group=Egg, parameterName=Egg Edge Highlight Amount}, B=@n3.RGB)
  n6 = BreakMaterialAttributes!(Attr=MaterialFunctionCall() {materialFunction=&/MF_Egg_Inside}.Result)
  n7 = ScalarParameter() {defaultValue=-0.2857, group=Dragon_Inside_Egg, parameterName=Dragon_Contrast}
  n8 = LinearInterpolate(
   A=Multiply(
    A=MaterialFunctionCall(In=@n6.BaseColor, Contrast=@n7) {materialFunction=&CheapContrast}.Result
    B=VectorParameter() {group=Egg, parameterName=EggShell_Inside_Color, defaultValue=<rgba 0.1382 0.3551 0.3698 1>}.RGB
   )
   B=VectorParameter() {group=Egg, parameterName=EggShell_Outside_Color, defaultValue=<rgba 0.7865 1 0.5729 1>}.RGB
   Alpha=Fresnel(ExponentIn=ScalarParameter() {defaultValue=1.4207, group=Egg, parameterName=Egg Fresnel amount}, Normal=VertexNormalWS()) {exponent=1}
  )
  n9 = TextureSampleParameter2D() {group=Textures, parameterName=Param, samplerType=SAMPLERTYPE_Masks, texture=/Game/Environments/Props/Egg/Textures/T_Egg_S}
  n10 = SetMaterialAttributes!(
   MaterialAttributes
   Base Color=Multiply(A=@n1.RGB, B=Add(A=@n5, B=@n8))
   Metallic=@n9.G
   Specular=@n9.G
   Roughness=Multiply(A=ScalarParameter() {defaultValue=0.7, group=Egg, parameterName=Roughness Bias}, B=@n9.B)
   Emissive Color
   Subsurface Color
   Normal=@n2.RGB
  ) {attributeSetTypes [
    #1
    #2
    #3
    #4
    #5
    #6
    #7
   ]}
  n11 = OneMinus(None=@n9.A)
  n12 = Constant() {r=1}
  n13 = Constant()
  n14 = Constant() {r=1}
  n15 = TextureSampleParameter2D() {
   group=Snow
   parameterName=Snow_Normal
   samplerType=SAMPLERTYPE_Normal
   texture=/Game/Athena/Environments/Landscape/Textures/T_Athena_Snow_01_N
  }
  n16 = ComponentMask(None=Transform(None=@n2.RGB)) {b=true}
  n17 = Multiply(
   A=MaterialFunctionCall(
    Transition Phase=ScalarParameter() {defaultValue=0.3894, group=Snow, parameterName=Snow}
    Height Texture=@n1.A
    Contrast=ScalarParameter() {defaultValue=-0.432, group=Snow, parameterName=Snow Contrast}
   ) {materialFunction=&HeightLerp}.Alpha
   B=ScalarParameter() {defaultValue=1, group=Snow, parameterName=SnowUpVector}
  )
  n18 = Clamp(
   None=Saturate(None=Multiply(A=@n16, B=@n17))
   Min=ScalarParameter() {defaultValue=-18.7563, group=Snow, parameterName=Snow Clamp Min}
   Max=ScalarParameter() {defaultValue=0.9619, group=Snow, parameterName=Snow Clamp Max}
  )
  n19 = MaterialFunctionCall(In=Add(A=@n18, B=@n18), Contrast=ScalarParameter() {defaultValue=0.8083, group=Snow, parameterName=Snow Mask Contrast}) {materialFunction=&CheapContrast}
  n20 = WorldPosition()
  n21 = ComponentMask(None=Divide(A=@n20.XYZ, B=ScalarParameter() {defaultValue=24.595, group=Graphical Blend Mask, parameterName=B})) {g=true, r=true}
  n22 = TextureSample(UVs=@n21) {samplerType=SAMPLERTYPE_LinearGrayscale, texture=/Game/Global/Textures/Default/Noise/T_SwirlNoise_H2}
  n23 = TextureSample(UVs=Divide(A=@n21, B=ScalarParameter() {defaultValue=3, group=Graphical Blend Mask, parameterName=A})) {samplerType=SAMPLERTYPE_LinearGrayscale, texture=&T_SwirlNoise_H2}
  n24 = Saturate(None=Divide(
    A=MaterialFunctionCall!(World Position=@n20.XYZ) {materialFunction=/Engine/Functions/Engine_MaterialFunctions02/Utility/DepthFromWorldPosition}.Result
    B=ScalarParameter() {defaultValue=635.8153, group=Graphical Blend Mask, parameterName=depth}
   ))
  n25 = ComponentMask(None=Multiply(
    A=LinearInterpolate(A=@n22.R, B=@n23.RGB, Alpha=@n24)
    B=ScalarParameter() {defaultValue=1.183, group=Graphical Blend Mask, parameterName=SnowEdgeContrast}
   )) {r=true}
  MaterialFunctionCall(MA In=MaterialFunctionCall!(
    Base Material=MaterialFunctionCall!(Base Material=@n4.Result, Top Material=@n10, Alpha=@n11) {materialFunction=/Engine/Functions/MaterialLayerFunctions/MatLayerBlend_Standard}.Blended Material
    Top Material=MakeMaterialAttributes!(
     BaseColor=@n12
     Metallic=@n13
     Specular
     Roughness=@n14
     Anisotropy
     EmissiveColor
     Opacity
     OpacityMask
     Normal=@n15.RGB
     Tangent
     WorldPositionOffset
     SubsurfaceColor
     ClearCoat
     ClearCoatRoughness
     AmbientOcclusion
     Refraction
     CustomizedUVs_0
     CustomizedUVs_1
     CustomizedUVs_2
     CustomizedUVs_3
     CustomizedUVs_4
     CustomizedUVs_5
     CustomizedUVs_6
     CustomizedUVs_7
     PixelDepthOffset
     ShadingModel
     Displacement
    )
    Alpha=LinearInterpolate(B=@n19.Result, Alpha=@n25)
   ) {materialFunction=/Engine/Functions/MaterialLayerFunctions/MatLayerBlend_Simple}.Blended Material) {materialFunction=&MF_CreativeMoveTool}
 ]
}
